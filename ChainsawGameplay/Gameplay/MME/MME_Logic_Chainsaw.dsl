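; Movement mode extension logic for Chainsaw "scooter" movement, tied to the Scootering gameplay tag.
; On activation, the dedicated server registers the equipped weapon as activating item for sprint durability cost.
; De-nativized Blueprint functions steer velocity toward input or control direction, easing max speed for obstacles or sharp turns, and send ledge-stop events.
; Behind bDeNativize, push tracks obstacle tags into runtime data and end sends a stop event.
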(event EventOnExtensionPush
  (params (RuntimeContext :type MMERuntime%20Context%20Structure%20%28by%20ref%29))
  (if (Variables|Chainsaw|GetbDeNativize)
    (Development|PrintString 0 "PUSH!" true true "(R=0.000000,G=0.041645,B=1.000000,A=1.000000)" 1e+01)
    (bind _returnvalue (Class|FortMovementModeExtBlueprintLibrary|GetRuntimeDatafromContext RuntimeContext "/ChainsawGameplay/Gameplay/MME/MME_Runtime_Chainsaw.MME_Runtime_Chainsaw_C"))
    (bind _output_get (Variables|Default|SetRuntimeData _returnvalue))
    (Variables|Default|SetTEMP_GoalMaxSpeed 1e+03)
    (Variables|Default|SetTEMP_CurrentMaxSpeed 1e+03)
    (Ability|Tasks|WaitGameplayTagAddToActor (Variables|FortMovementMode_BaseExtLogic|GetOwnerPawn) "(TagName=\"Event.Athena.Chainsaw.ScooterBlockedByObstacle\")"
      (:then
        (Ability|Tasks|WaitGameplayTagRemoveFromActor (Variables|FortMovementMode_BaseExtLogic|GetOwnerPawn) "(TagName=\"Event.Athena.Chainsaw.ScooterBlockedByObstacle\")"
          (:Removed
            (Variables|/ChainsawGameplay/Gameplay/MME/MME_Runtime_Chainsaw.MME_Runtime_Chainsaw_C|SetBlockedByObstacle false (Variables|Default|GetRuntimeData))
            (Variables|Default|SetTEMP_BlockedByObstacle false))))
      (:Added
        (Variables|Default|SetTEMP_BlockedByObstacle true)
        (Variables|/ChainsawGameplay/Gameplay/MME/MME_Runtime_Chainsaw.MME_Runtime_Chainsaw_C|SetBlockedByObstacle true (Variables|Default|GetRuntimeData))))))

(event Custom|CheckIfAbilityIsActive
  (if (not (GameplayTags|HasMatchingGameplayTag (Variables|FortMovementMode_BaseExtLogic|GetOwnerPawn) "(TagName=\"Granted.Athena.Chainsaw.Scootering\")"))
    (Class|FortMovementCompCharacter|CancelActiveExtensionwithTag (Variables|FortMovementMode_BaseExtLogic|GetOwnerMovementComponent) (Variables|FortMovementMode_BaseExtLogic|GetMovementModeExtensionIDTag))))

(event EventOnExtensionEnd
  (params (RuntimeContext :type MMERuntime%20Context%20Structure%20%28by%20ref%29))
  (if (Variables|Chainsaw|GetbDeNativize)
    (Development|PrintString 0 "END!" true true "(R=0.000000,G=0.041645,B=1.000000,A=1.000000)" 1e+01)
    (Ability|SendGameplayEventtoActor (Variables|FortMovementMode_BaseExtLogic|GetOwnerPawn) "(TagName=\"Event.Athena.Chainsaw.StopSawScooter\")") ; End ability if it isn't already ended
  ))

(event EventOnExtensionActivate
  (params (RuntimeContext :type MMERuntime%20Context%20Structure%20%28by%20ref%29))
  ;[0.000242 0.08174 0.15 0.5] OnExtensionActivate: Update the activating item so that durability cost is reducted when sprint is used.
  (bind _ownerpawn (Variables|FortMovementMode_BaseExtLogic|GetOwnerPawn)
    (:then
      (if (Networking|IsDedicatedServer _ownerpawn)
        (bind _returnvalue (Class|FortMovementModeExtBlueprintLibrary|GetRuntimeDatafromContext RuntimeContext "/Script/FortniteGame.FortMovementMode_ItemActivatedRuntimeData"))
        (Utilities|IsValid _returnvalue
          (:Is%20Valid
            (if (GameplayTags|HasMatchingGameplayTag _ownerpawn (Variables|Default|GetWeaponEquippedTag))
              (Class|FortMovementModeItemActivatedRuntimeData|SetActivatingItemfromWeapon _returnvalue (Variables|/Script/FortniteGame.FortPawn|GetCurrentWeapon _ownerpawn))
              (else
                (Fortnite|FortShippingLog self "MME_Logic_Chainsaw activated but the Equipped tag is missing from the OwnerPawn! Cannot set activating item from weapon reliably." true)))))))
    (:else)))

(fn PhysUpdate(De-Nativized)
  (params (RuntimeContext :type MMERuntime%20Context%20Structure%20%28by%20ref%29) (InVelocity :type Vector%20%28by%20ref%29 :default "0, 0, 0") (InAcceleration :type Vector%20%28by%20ref%29 :default "0, 0, 0"))
  (bind (_bblockinghit _bwalkablefloor) (Utilities|Struct|BreakFindFloorResult (Pawn|Components|CharacterMovement|FindFloor (Variables|/Script/Engine.Character|GetCharacterMovement (Variables|FortMovementMode_BaseExtLogic|GetOwnerPawn)) (Transformation|GetWorldLocation (Variables|/Script/Engine.Character|GetCapsuleComponent (Variables|FortMovementMode_BaseExtLogic|GetOwnerPawn))))))
  (bind _returnvalue (Class|FortMovementModeExtBlueprintLibrary|GetRuntimeDatafromContext RuntimeContext "/ChainsawGameplay/Gameplay/MME/MME_Runtime_Chainsaw.MME_Runtime_Chainsaw_C"))
  (Variables|Default|SetRuntimeData _returnvalue)
  (LocalVariables|SetDeltaTime (Class|FortMovementModeExtBlueprintLibrary|GetDeltaTimefromContext RuntimeContext))
  (Utilities|FlowControl|Sequence
    (:then_0
      (if (not _bwalkablefloor)
        (Ability|SendGameplayEventtoActor (Variables|FortMovementMode_BaseExtLogic|GetOwnerPawn) "(TagName=\"Event.Athena.Chainsaw.StopSawScooter.FromLedge\")")))
    (:then_2
      (Variables|Default|SetTEMP_CurrentMaxSpeed (Math|Interpolation|FInterpTo (Variables|Default|GetTEMP_CurrentMaxSpeed) (Math|Float|SelectFloat 2e+02 (Math|Float|SelectFloat 5e+02 (Variables|Default|GetTEMP_GoalMaxSpeed) (Variables|Default|GetTEMP_DoingSharpTurn)) (Variables|Default|GetTEMP_BlockedByObstacle)) (LocalVariables|GetDeltaTime) (Math|Float|SelectFloat 15.0 2.0 (Variables|Default|GetTEMP_BlockedByObstacle))))
      (Variables|Default|SetTEMP_GoalMaxSpeed 1.5e+03))
    (:then_4
      (LocalVariables|SetNewVelocity (Math|Vector|VectorClampSize2D (+ (Utilities|Operators|Multiply (CallFunction|GoalDirection (Variables|GetInAcceleration)) (LocalVariables|GetDeltaTime) (Variables|Default|GetTEMP_TurnSpeed)) (Variables|GetInVelocity)) 0.0 (Variables|Default|GetTEMP_CurrentMaxSpeed))))
    (:then_5
      (return (LocalVariables|GetNewVelocity) (Variables|GetInAcceleration)))))

(fn UpdateBeforeCharacterMovement(De-Nativized)
  (params (RuntimeContext :type MMERuntime%20Context%20Structure%20%28by%20ref%29))
  (if (GameplayTags|HasMatchingGameplayTag (Variables|FortMovementMode_BaseExtLogic|GetOwnerPawn) "(TagName=\"Granted.Athena.Chainsaw.Scootering\")")
    (return "Active")
    (else
      (return "Ended"))))

(fn DotBetweenVelocityAndGoalVelocity
  (params (Acceleration :type Vector) (InVelocity :type Vector))
  (return (Math|Vector|DotProduct (Math|Vector|Normalize2D(Vector) (CallFunction|GoalDirection Acceleration)) (Math|Vector|Normalize2D(Vector) InVelocity))))

(fn DoTraceForward)

(fn GetMovementDirectionSafe
  (return "0, 0, 0"))

(fn GoalDirection
  (params (Acceleration :type Vector))
  (return (Math|Vector|SelectVector (Math|Vector|Normalize Acceleration) (Math|Vector|GetForwardVector (Pawn|GetDesiredRotation (Pawn|GetController (Variables|FortMovementMode_BaseExtLogic|GetOwnerPawn)))) (> (Math|Vector|VectorLength Acceleration) 3e+02))))
